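# Animation Blueprint for the M_MED_Jupiter male player body that corrects dynamic shoulder, chest and arm socket bones on top of a copied character pose.
# It copies the attached parent's pose, additively scales both arm sockets, then rotation-constrains shoulder, clavicle-chest and chest bones between neighboring spine, clavicle and upper-arm bones.
# Socket scale constraints and Scale_Arm_Socket getters sit unconnected; initialization only caches the mesh component, leaving the scale-delta calculation unexecuted.
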AnimBlueprint /BRCosmetics/Characters/Player/Male/Medium/Bodies/M_MED_Jupiter/Meshes/M_MED_Jupiter_AnimBP
target_skeleton=&/M_MED_Jupiter_Skeleton
graphs [
 AnimGraph {nodes [
   "Constraint - dyn_arm_socket_l (SS)"!(ComponentPose, Alpha=1, ConstraintWeights_0=0.9, ConstraintWeights_1:float=0.1) -> (Pose) {
    alphaBoolBlend {blendOption=Linear}
    boneToModify {boneName=dyn_arm_socket_l}
    constraintSetup [
     {offsetOption=None, transformType=Scale, targetBone {boneName=dyn_clav_chest_l}}
     {offsetOption=None, transformType=Scale, targetBone {boneName=spine_05}}
    ]
    constraintWeights [1,1]
   }
   "Constraint - dyn_arm_socket_r (SS)"!(ComponentPose, Alpha=1, ConstraintWeights_0=0.9, ConstraintWeights_1:float=0.1) -> (Pose) {
    alphaBoolBlend {blendOption=Linear}
    boneToModify {boneName=dyn_arm_socket_r}
    constraintSetup [{transformType=Scale, targetBone {boneName=dyn_clav_chest_r}}, {transformType=Scale, targetBone {boneName=spine_05}}]
    constraintWeights [1,1]
   }
   n1 = "Transform (Modify) Bone - Bone: dyn_arm_socket_r"(
    ComponentPose=Local To Component(LocalPose=Copy Pose From Mesh() {bUseAttachedParent=true})
    Alpha=1
    Translation="0,0,0"
    Rotation="0,0,0"
    Scale="0.1,0.1,0.1"
   ) {
    scaleMode=BMM_Additive
    scaleSpace=BCS_BoneSpace
    alphaBoolBlend {blendOption=Linear}
    boneToModify {boneName=dyn_arm_socket_r}
    scale=<0.1 0.1 0.1>
   }
   n2 = "Transform (Modify) Bone - Bone: dyn_arm_socket_l"(ComponentPose=@n1, Alpha=1, Translation="0,0,0", Rotation="0,0,0", Scale="0.1,0.1,0.1") {
    scaleMode=BMM_Additive
    scaleSpace=BCS_BoneSpace
    alphaBoolBlend {blendOption=Linear}
    boneToModify {boneName=dyn_arm_socket_l}
    scale=<0.1 0.1 0.1>
   }
   n3 = "Constraint - dyn_shoulder_r (RR)"!(ComponentPose=@n2, Alpha=1, ConstraintWeights_0=0.5, ConstraintWeights_1:float=0.5) -> (Pose) {
    alphaBoolBlend {blendOption=Linear}
    boneToModify {boneName=dyn_shoulder_r}
    constraintSetup [{transformType=Rotation, targetBone {boneName=clavicle_r}}, {transformType=Rotation, targetBone {boneName=upperarm_r}}]
    constraintWeights [1,1]
   }
   n4 = "Constraint - dyn_clav_chest_r (RR)"!(ComponentPose=@n3, Alpha=1, ConstraintWeights_0=0.3, ConstraintWeights_1:float=0.7) -> (Pose) {
    alphaBoolBlend {blendOption=Linear}
    boneToModify {boneName=dyn_clav_chest_r}
    constraintSetup [{transformType=Rotation, targetBone {boneName=spine_05}}, {transformType=Rotation, targetBone {boneName=clavicle_r}}]
    constraintWeights [1,1]
   }
   n5 = "Constraint - dyn_shoulder_l (RR)"!(ComponentPose=@n4, Alpha=1, ConstraintWeights_0=0.5, ConstraintWeights_1:float=0.5) -> (Pose) {
    alphaBoolBlend {blendOption=Linear}
    boneToModify {boneName=dyn_shoulder_l}
    constraintSetup [{transformType=Rotation, targetBone {boneName=clavicle_l}}, {transformType=Rotation, targetBone {boneName=upperarm_l}}]
    constraintWeights [1,1]
   }
   n6 = "Constraint - dyn_clav_chest_l (RR)"!(ComponentPose=@n5, Alpha=1, ConstraintWeights_0=0.2, ConstraintWeights_1:float=0.8) -> (Pose) {
    alphaBoolBlend {blendOption=Linear}
    boneToModify {boneName=dyn_clav_chest_l}
    constraintSetup [{transformType=Rotation, targetBone {boneName=spine_05}}, {transformType=Rotation, targetBone {boneName=clavicle_l}}]
    constraintWeights [1,1]
   }
   n7 = "Constraint - dyn_chest (RR)"!(ComponentPose=@n6, Alpha=1, ConstraintWeights_0=0.2, ConstraintWeights_1:float=0.8) -> (Pose) {
    alphaBoolBlend {blendOption=Linear}
    boneToModify {boneName=dyn_chest}
    constraintSetup [{transformType=Rotation, targetBone {boneName=spine_05}}, {transformType=Rotation, targetBone {boneName=spine_04}}]
    constraintWeights [1,1]
   }
   Output Pose(Result=Component To Local(ComponentPose=@n7)) {layerGroup=None}
   Get Scale_Arm_Socket() -> (Scale_Arm_Socket:double)
   Get Scale_Arm_Socket()
  ]}
 EventGraph {nodes [
   Event BlueprintInitializeAnimation()
   > Set Skeletal Mesh(Skeletal Mesh:Skeletal Mesh Component Object Reference=GetOwnerSkeletalMeshComponent()) -> (Output_Get:Skeletal Mesh Component Object Reference)
   Set Scale_Arm_Socket(Scale_Arm_Socket:double=GetAlphaFromReferenceTransformDelta(
     SkeletalMeshComponent=Get Skeletal Mesh() -> (Skeletal Mesh:Skeletal Mesh Component Object Reference)
     BoneName=dyn_arm_socket_r
     BaseName=dyn_clav_chest_r
     DeltaComponentType=Scale
     InputRange=(X=10,Y=10)
     TwistAxis="0,0,0"
    )) -> (Output_Get:double)
  ]}
]
variables [
 Scale_Arm_Socket {type=number}
 Skel Mesh {default=""}
 Skeletal Mesh {type=SkeletalMeshComponent}
]
